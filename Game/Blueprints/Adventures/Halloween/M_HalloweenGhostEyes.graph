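# Additive unlit particle material, likely for glowing Halloween ghost eyes, as the asset name suggests.
# It multiplies particle color by a grayscale camera-flare texture, boosts the result, and feeds it into the emissive input of the MF_FX_EmissiveAdjustments function.
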Material /Game/Blueprints/Adventures/Halloween/M_HalloweenGhostEyes
graph {
 properties {bUsedWithParticleSprites=true, blendMode=BLEND_Additive, pixelDepthOffsetMode=PDOM_Legacy, shadingModel=MSM_Unlit, usedShadingModels=Unlit}
 nodes [
  n1 = ParticleColor()
  n2 = TextureSample() {samplerType=SAMPLERTYPE_Grayscale, texture=/Game/Effects/Fort_Effects/Textures/T_EFX_CameraFlare_01}
  MaterialFunctionCall(Emissive=Multiply(A=Multiply(A=@n1.RGB, B=@n2.RGB)) {constB=4}) {materialFunction=&MF_FX_EmissiveAdjustments}
 ]
}
